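# Sound cue for the whoosh outro in the ranked front-end menu, played in the Fort_UI sound class at reduced volume.
# A single wave player feeds the outro whoosh wave through a Modulator node left at default settings.
# A concurrency override caps it at two simultaneous instances, stopping the oldest when exceeded.
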SoundCue /Game/Sounds/Fort_UI_Sounds/FrontEnd/New_FrontEnd/Ranked/Fort_UI_Ranked_Whoosh_Outro_Cue
root=Modulator_0
nodes [
 {id=Modulator_0, children [WavePlayer_1]}
 {id=WavePlayer_1, properties {soundWaveAssetPtr=/Game/Sounds/Fort_UI_Sounds/FrontEnd/New_FrontEnd/Ranked/wav/UI_Ranked_Menu_Outro_Whoosh}}
]
properties {
 bOverrideConcurrency=true
 maxDistance=2097152
 soundClassObject=&Fort_UI
 volumeMultiplier=0.3
 concurrencyOverrides {maxCount=2, resolutionRule=StopOldest}
}
