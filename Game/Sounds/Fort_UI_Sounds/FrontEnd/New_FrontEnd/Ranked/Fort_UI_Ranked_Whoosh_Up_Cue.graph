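# Sound cue for an upward whoosh in the ranked front-end UI, playing the display toggle whoosh wave.
# A single wave player feeds a modulator at the root, which applies its default random pitch and volume variation.
# It plays quietly in the UI sound class, with an overridden concurrency that caps simultaneous copies and stops the oldest.
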SoundCue /Game/Sounds/Fort_UI_Sounds/FrontEnd/New_FrontEnd/Ranked/Fort_UI_Ranked_Whoosh_Up_Cue
root=Modulator_1
nodes [
 {id=Modulator_1, children [WavePlayer_0]}
 {id=WavePlayer_0, properties {soundWaveAssetPtr=/Game/Sounds/Fort_UI_Sounds/FrontEnd/New_FrontEnd/Ranked/wav/UI_Ranked_DisplayToggleWhoosh}}
]
properties {
 bOverrideConcurrency=true
 maxDistance=2097152
 soundClassObject=&Fort_UI
 volumeMultiplier=0.12
 concurrencyOverrides {maxCount=2, resolutionRule=StopOldest}
}
