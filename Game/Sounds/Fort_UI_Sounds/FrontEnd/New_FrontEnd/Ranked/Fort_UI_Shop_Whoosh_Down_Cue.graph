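# UI sound cue for a downward shop whoosh, played on the Fort_UI sound class at a low volume.
# A Random node picks with equal weight between two Modulators, each fed by a player of the same UI_Ranked_DisplayToggleWhoosh wave.
# One branch fixes the pitch lower while the other keeps default modulation, giving two variations from one asset.
# A concurrency override caps it at two instances, stopping the oldest.
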SoundCue /Game/Sounds/Fort_UI_Sounds/FrontEnd/New_FrontEnd/Ranked/Fort_UI_Shop_Whoosh_Down_Cue
root=Random_0
nodes [
 {id=Modulator_1, children [WavePlayer_0], properties {pitchMax=0.7, pitchMin=0.7}}
 {id=Modulator_2, children [WavePlayer_2]}
 {id=Random_0, children [Modulator_2,Modulator_1], properties {weights [1,1]}}
 {id=WavePlayer_0, properties {soundWaveAssetPtr=/Game/Sounds/Fort_UI_Sounds/FrontEnd/New_FrontEnd/Ranked/wav/UI_Ranked_DisplayToggleWhoosh}}
 {id=WavePlayer_2, properties {soundWaveAssetPtr=&UI_Ranked_DisplayToggleWhoosh}}
]
properties {
 bOverrideConcurrency=true
 maxDistance=2097152
 soundClassObject=&Fort_UI
 volumeMultiplier=0.12
 concurrencyOverrides {maxCount=2, resolutionRule=StopOldest}
}
