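# Sound cue for the back button in Fortnite's Ranked front-end UI, routed to the Fort_UI sound class.
# A mixer layers three waves: a full-volume tile select, then a Ranked close whoosh and a Battle Pass modal whoosh, both much quieter and staggered by short fixed delays.
# Modulators pitch every layer down, with slight random variation on the select and close whoosh.
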SoundCue /Game/Sounds/Fort_UI_Sounds/FrontEnd/New_FrontEnd/Ranked/Fortnite_UI_Ranked_Button_Back_Cue
root=Mixer_0
nodes [
 {id=Delay_0, children [WavePlayer_1], properties {delayMax=0.1, delayMin=0.1}}
 {id=Delay_1, children [WavePlayer_2], properties {delayMax=0.2, delayMin=0.2}}
 {id=Mixer_0, children [Modulator_2,Modulator_0,Modulator_1], properties {inputVolume [1,1,1]}}
 {id=Modulator_0, children [Delay_0], properties {pitchMax=0.8, pitchMin=0.7, volumeMax=0.2, volumeMin=0.2}}
 {id=Modulator_1, children [Delay_1], properties {pitchMax=0.8, pitchMin=0.8, volumeMax=0.2, volumeMin=0.2}}
 {id=Modulator_2, children [WavePlayer_0], properties {pitchMax=0.8, pitchMin=0.7, volumeMax=1, volumeMin=1}}
 {id=WavePlayer_0, properties {soundWaveAssetPtr=/Game/Sounds/Fort_UI_Sounds/FrontEnd/New_FrontEnd/Ranked/wav/UI_Ranked_Select_Tile}}
 {id=WavePlayer_1, properties {soundWaveAssetPtr=/Game/Sounds/Fort_UI_Sounds/FrontEnd/New_FrontEnd/Ranked/wav/UI_Ranked_CloseWhoosh}}
 {id=WavePlayer_2, properties {soundWaveAssetPtr=/Game/Sounds/Fort_UI_Sounds/FrontEnd/BattlePass/BP_ToolKit/BP_Modal_Whoosh_Med_03}}
]
properties {maxDistance=2097152, soundClassObject=&Fort_UI, volumeMultiplier=0.8}
