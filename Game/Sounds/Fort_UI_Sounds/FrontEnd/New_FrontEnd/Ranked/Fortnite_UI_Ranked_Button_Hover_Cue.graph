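# Sound cue for hovering over a button in the Ranked front-end UI.
# A single wave player plays the shared default UI hover sound, routed through the Fort_UI sound class with virtualization disabled.
# Override concurrency allows one instance, stopping the oldest with releasable volume scaling.
# A shared hover-tick concurrency asset is also assigned, though the override flag is on.
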SoundCue /Game/Sounds/Fort_UI_Sounds/FrontEnd/New_FrontEnd/Ranked/Fortnite_UI_Ranked_Button_Hover_Cue
root=WavePlayer_0
nodes [{id=WavePlayer_0, properties {soundWaveAssetPtr=/Game/Sounds/Fort_UI_Sounds/FrontEnd/DefaultSounds/PressedButtons/wav/UI_MP_Hover_Default_01}}]
properties {
 bOverrideConcurrency=true
 maxDistance=2097152
 soundClassObject=&Fort_UI
 virtualizationMode=Disabled
 volumeMultiplier=1
 concurrencyOverrides {bVolumeScaleCanRelease=true, maxCount=1, resolutionRule=StopOldest, volumeScale=0.5, volumeScaleAttackTime=0.25, volumeScaleReleaseTime=0.25}
 concurrencySet [/Game/Sounds/Concurrency_Rules/FrontEnd_UI_Menu_Max_01_Stop_Oldest_Hover_Tick]
}
